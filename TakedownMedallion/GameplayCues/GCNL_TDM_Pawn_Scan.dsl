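; Looping Gameplay Cue notify for the Takedown Medallion's pawn scan effect.
; When looping starts, it safely takes the first spawned particle component from the cue's array.
; It sets that component's scanner-center vector parameter, named by a variable, to the instigator's location from the cue parameters.
; The center is written only at loop start, and the construction script is empty.
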(event EventOnLoopingStart(EitherVFX-preferthisversion)
  (params (MyTarget :type Actor%20Object%20Reference) (Parameters :type Gameplay%20Cue%20Parameters%20Structure%20%28by%20ref%29) (ParticleComponents :type Array%20of%20FXSystem%20Component%20Object%20References) (AudioComponents :type Array%20of%20Audio%20Component%20Object%20References))
  (bind (_normalizedmagnitude _rawmagnitude _effectcontext _matchedtagname _originaltag _aggregatedsourcetags _aggregatedtargettags _location _normal _instigator) (Ability|GameplayCue|BreakGameplayCueParameters Parameters))
  (bind _result (ArrayMacros|SafeArrayGet ParticleComponents)
    (:%20%20
      (Effects|Components|ParticleSystem|SetVectorParameter _result (Variables|Default|GetScannerCenterParam) (Transformation|GetActorLocation _instigator)))))

(fn ConstructionScript)
